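# Sound cue for one Optimus radio voice-over line in the Season 25 Battle Pass content.
# A single dialogue player node is the root and plays the line's dialogue wave, with the AloeCrouton_OP_DV voice as speaker and no targets.
# The cue routes through the narrator VO sound class and uses the shared Battle Pass audio-log attenuation settings.
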SoundCue /BattlePassS25/Sound/VO/Cues/OPTIMUS_2023_RadioVO_01_02_010_Dialogue_Cue
root=DialoguePlayer_0
nodes [{
  id=DialoguePlayer_0
  properties {dialogueWaveParameter {
    dialogueWave=/BattlePassS25/Sound/VO/Waves/AloeCrouton/OPTIMUS_2023_RadioVO_01_02_010_Dialogue
    context {speaker=/BattlePassS25/Sound/VO/AloeCrouton_OP_DV, targets []}
   }}
 }]
properties {
 attenuationSettings=/BattlePassBase/Audio/SFX/Fort_AudioLogs_Att
 maxDistance=2097152
 soundClassObject=/Game/Sounds/FortSoundClassesAndModes/Fort_VO_Narrator
 volumeMultiplier=1
}
