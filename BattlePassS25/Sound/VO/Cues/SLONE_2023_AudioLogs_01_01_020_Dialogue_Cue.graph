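# Sound cue that plays one Slone audio log voice line for Battle Pass S25.
# A single dialogue player node is the root, playing the line's dialogue wave with the Slone S25 voice as speaker and no targets.
# The cue uses the shared audio log attenuation from BattlePassBase and routes to the narrator VO sound class.
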SoundCue /BattlePassS25/Sound/VO/Cues/SLONE_2023_AudioLogs_01_01_020_Dialogue_Cue
root=DialoguePlayer_0
nodes [{
  id=DialoguePlayer_0
  properties {dialogueWaveParameter {
    dialogueWave=/BattlePassS25/Sound/VO/Waves/SLONE_2023_AudioLogs_01_01_020_Dialogue
    context {speaker=/BattlePassS25/Sound/VO/SloneS25_DV, targets []}
   }}
 }]
properties {
 attenuationSettings=/BattlePassBase/Audio/SFX/Fort_AudioLogs_Att
 maxDistance=2097152
 soundClassObject=/Game/Sounds/FortSoundClassesAndModes/Fort_VO_Narrator
 volumeMultiplier=1
}
